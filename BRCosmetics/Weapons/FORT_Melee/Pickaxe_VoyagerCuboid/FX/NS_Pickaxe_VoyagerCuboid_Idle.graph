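# Idle Niagara effect for the VoyagerCuboid pickaxe: an orange glowing cube with drifting small cubes and crackling lightning.
# MainCube combines mesh, pulsing glow sprite and light, scaled by the sampled skeletal bone or socket.
# SmallCubes spin, bob and rise within a rotating sphere, while three mesh lightning emitters add short-lived arcs, including ones around the sampled socket.
# Front-end flags and time-of-day inputs adjust color and brightness.
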NiagaraSystem /BRCosmetics/Weapons/FORT_Melee/Pickaxe_VoyagerCuboid/FX/NS_Pickaxe_VoyagerCuboid_Idle
dependencies {
 usedDataInterfaces [
  &NiagaraDataInterfaceMeshRendererInfo
  &NiagaraDataInterfaceCurve
  &NiagaraDataInterfaceVectorCurve
  &NiagaraDataInterfaceColorCurve
  &NiagaraDataInterfaceArrayFloat
  &NiagaraDataInterfaceSkeletalMesh
 ]
 usedDynamicInputs []
 usedModules [
  &self:SystemSpawnScript.SystemScriptSource.SystemScriptGraph.NiagaraNodeAssignment_0.SetVariables_#1
  &self:SystemSpawnScript.SystemScriptSource.SystemScriptGraph.NiagaraNodeAssignment_1.SetVariables_#2
  &self:SystemSpawnScript.SystemScriptSource.SystemScriptGraph.NiagaraNodeAssignment_2.SetVariables_#3
  &SystemState
 ]
}
emitters {
 Hunt_Lightning001 {
  emitterUpdateScript [
   &EmitterState(Life Cycle Mode=::NewEnumerator1, Loop Behavior=::NewEnumerator1, Loop Duration Mode=::NewEnumerator1, UseLoopDelay=false)
   &SpawnRate(SpawnRate=9, Use Spawn Probability=false)
  ]
  particleSpawnScript [
   &InitializeParticle(
    Color=<rgba 1 0.3027 0 1>
    Lifetime Max=0.55
    Lifetime Min=0.35
    Lifetime Mode=::NewEnumerator1
    Mesh Scale=&Multiply_VectorByFloat(
     Vector=&Lerp_Vector(
      A=&RandomRangeVector(Minimum=<0.5 0.4 0.4>, Maximum=<0.75 0.6 0.4>)
      B=&RandomRangeVector(Minimum=<1 1 0.4>, Maximum=<1 1 0.4>)
      Alpha=&FloatFromCurve(
       FloatCurve=DataInterface() {propertyValues {Curve {keys [{interpMode=RCIM_Cubic, tangentMode=RCTM_User, time=0.5, value=0}, <at 1 1>]}}}
       CurveIndex=&RandomRangeFloat(Minimum=0, Maximum=1)
      )
     )
     Float=0.2
    )
    Mesh Scale Mode=::NewEnumerator4
    UsePositionOffset=false
    Write Material Random=true
   )
   &InitialMeshOrientation(
    Mesh Axis To Orient=<-1 0 0>
    Mesh Orientation Mode=::NewEnumerator2
    Rotation=&RandomRangeVector(Minimum=<-0.07 -0.07 -0.07>, Maximum=<0.07 0.07 0.07>)
    Use Rotation Vector=true
   ) {enabled=false, module=InitialMeshOrientation002}
   /SetVariables(Particles.Random=&RandomRangeFloat(Minimum=0, Maximum=1)) {isSetParameters=true}
   &CameraOffset(Camera Offset Amount=10)
   &InitialMeshOrientation(
    Mesh Axis To Orient=<0 0 0>
    Mesh Orientation Mode=::NewEnumerator2
    Orientation Vector=&RotateVectorByQuaternion(
     Vector To Rotate=&SubtractVectorFromPosition(
      Position=&AddVectorToPosition(Position=$Particles.SkeletalMeshLocation.SampledPosition, Vector=<0 0 5>)
      Vector=Linked() {linkedVariable {name=Particles.Position, type=NiagaraPosition}}
     )
     Quaternion=$Particles.SkeletalMeshLocation.SampledRotation
    )
    Rotation=<0 0.5 0>
    Use Rotation Vector=true
   ) {enabled=false}
   &AddVelocity(
    Cone Angle=70
    Cone Axis=<0 1 0>
    Module.UseConeFalloff=false
    Module.UseVelDistribution=false
    Override Seed=false
    Velocity Mode=::NewEnumerator2
    Velocity Speed=1
   ) {enabled=false}
   &InitialMeshOrientation(
    Mesh Axis To Orient=<-1 0 0>
    Mesh Orientation Mode=::NewEnumerator2
    Orientation Vector=&RotateVectorByQuaternion(
     Vector To Rotate=&SubtractVectorFromPosition(
      Position=&AddVectorToPosition(Position=$Engine.Emitter.SimulationPosition, Vector=<0 0 5>)
      Vector=Linked() {linkedVariable {name=Particles.Position, type=NiagaraPosition}}
     )
     Quaternion=<0 0 0 1>
    )
    Rotation=<0 0.5 0>
    Use Rotation Vector=true
   ) {enabled=false, module=InitialMeshOrientation001}
  ]
  particleUpdateScript [
   &ParticleState()
   &SkeletalMeshLocation(
    Bone / Socket Index=1
    Bone Sampling Mode=::NewEnumerator4
    MaskSpawnGroup=false
    Orientation Sampling=::NewEnumerator1
    Skeletal Mesh=$System.Skeletal Mesh Reader
   )
   &ShapeLocation(
    Cylinder Height=1
    Cylinder Radius=7
    MaskSpawnGroup=false
    Offset=<0 0 -5>
    Offset Mode=::NewEnumerator0
    Override Seed=false
    Shape Origin=$Particles.SkeletalMeshLocation.SampledPosition
    Shape Primitive=::NewEnumerator1
    Surface Only=false
    UseOwnerScale=true
   )
   &UpdateMeshOrientation(
    Facing Direction=&RotateVectorByQuaternion(
     Vector To Rotate=&SubtractVectorFromPosition(
      Position=&AddVectorToPosition(Position=$Particles.SkeletalMeshLocation.SampledPosition, Vector=<0 0 5>)
      Vector=Linked() {linkedVariable {name=Particles.Position, type=NiagaraPosition}}
     )
     Quaternion=$Particles.SkeletalMeshLocation.SampledRotation
    )
    Facing Interpolation=false
    Orientation Method=::NewEnumerator1
    Override Quat=false
    TransformFacing=false
    TransformUp=false
   ) {enabled=false}
   &ScaleMeshSize(Scale Factor=&Multiply_VectorByFloat(Vector=<1 1.5 0.5>, Float=1.5))
   &ScaleColor(
    Scale Alpha=&Multiply_Float(
     A=10
     B=&FloatFromCurve(FloatCurve=DataInterface() {propertyValues {Curve {keys [
          {interpMode=RCIM_Cubic, tangentMode=RCTM_User, time=0, value=0}
          {interpMode=RCIM_Cubic, tangentMode=RCTM_User, time=0.1, value=0}
          {interpMode=RCIM_Cubic, tangentMode=RCTM_User, time=0.2, value=0.8}
          {interpMode=RCIM_Cubic, tangentMode=RCTM_User, time=1, value=0}
         ]}}})
    )
    Scale RGB=&Multiply_VectorByFloat(
     Vector=&Multiply_VectorByFloat(Vector=<1 1 1>, Float=&FN_FrontEndFloatModifier(Gameplay=1, Front End=0.5))
     Float=&Multiply_Float(
      A=&RandomRangeFloat(Minimum=0.5, Maximum=1, Evaluation Type=::NewEnumerator1)
      B=&Multiply_Float(A=&FNTimeOfDayScalar(Night Scale=0.2, Morning Scale=1, Evening Scale=1), B=10)
     )
    )
    ScaleA=true
    ScaleRGB=true
   )
   &DynamicMaterialParameters(
    Index 0 Param 1=&FloatFromCurve(FloatCurve=DataInterface() {propertyValues {Curve {keys [{tangentMode=RCTM_User, time=0, value=0}, {tangentMode=RCTM_User, time=1, value=1}]}}})
    Index 0 Param 3=&FloatFromCurve(FloatCurve=DataInterface() {propertyValues {Curve {keys [
         {arriveTangent=4.0072, interpMode=RCIM_Cubic, leaveTangent=4.0072, tangentMode=RCTM_User, time=0, value=0}
         {arriveTangent=2.7083, interpMode=RCIM_Cubic, leaveTangent=2.7083, time=0.2543, value=0.6225}
         {arriveTangent=1.8473, interpMode=RCIM_Cubic, leaveTangent=1.8473, time=0.6, value=1.625}
         {arriveTangent=1.6993, interpMode=RCIM_Cubic, leaveTangent=1.6993, tangentMode=RCTM_User, time=1, value=2}
        ]}}})
    Index 0 Param 4=15
    Index 1 Param 1=0.8
    Index 1 Param 2=$Particles.Random
    Index 1 Param 3=1.3
    Index 1 Param 4=1
    Index 2 Param 1=&FloatFromCurve(FloatCurve=DataInterface() {propertyValues {Curve {keys [<at 0 1>,<at 1 0.3>]}}})
    Module.Index0Float0Write=true
    Module.Index0Float1Write=false
    Module.Index0Float2Write=true
    Module.Index0Float3Write=true
    Module.Index1Float0Write=true
    Module.Index1Float1Write=true
    Module.Index1Float2Write=true
    Module.Index1Float3Write=true
    Module.Index2Float0Write=true
    Module.Index2Float1Write=true
    Module.Index2Float2Write=true
    Module.Index2Float3Write=true
    Write Parameter Index 1=::NewEnumerator1
    Write Parameter Index 2=::NewEnumerator1
   )
   &SolveForcesAndVelocity(Clamp Velocity=false, Limit Acceleration=false) {enabled=false}
   /SetVariables(Particles.Lightning glow color:LinearColor=&MakeLinearColorFromVectorAndFloat(
     "Vector (RGB)"=&MakeVectorFromLinearColorRGB(Color=&MultiplyLinearColor(
       Linear Color=<rgba 4 1.4833 0 1>
       Scale Factor=&MakeVector4_FromFloat(Float=&RandomRangeFloat(Minimum=0.5, Maximum=1, Evaluation Type=::NewEnumerator1))
      ))
     "Float (Alpha)"=&Multiply_Float(
      A=&FloatFromCurve(FloatCurve=DataInterface() {propertyValues {Curve {keys [<at 0 1 RCIM_Cubic>, {interpMode=RCIM_Cubic, leaveTangent=-3.6307, time=0.9, value=0}]}}})
      B=5
     )
    )) {isSetParameters=true}
   /SetVariables(Particles.Glow Dy02:Vector4f=&MakeVector4(
     X=&FloatFromCurve(FloatCurve=DataInterface() {propertyValues {Curve {keys [{tangentMode=RCTM_User, time=0, value=0.2}, {tangentMode=RCTM_User, time=0.5, value=1}]}}})
     Y=0.7
     Z=0
     W=5
    )) {isSetParameters=true}
   /SetVariables(Particles.Glow Dy03:Vector4f=&MakeVector4(
     X=&FloatFromCurve(
      FloatCurve=DataInterface() {propertyValues {Curve {keys [<at 0 1>,<at 1 0.5>]}}}
      Scale Curve=&MaskFloat(Mask Source=1, Mask Parameter=$System.VariantIndex, Mask=2, Masked Float=0.7, Passthrough Float=0.5)
     )
     Y=$Particles.Random
     Z=1.3
     W=1
    )) {isSetParameters=true}
   /SetVariables(Particles.MeshOrientation=&OrientMeshToVector(
     Look At Direction=&RotateVectorByQuaternion(
      Vector To Rotate=&SubtractVectorFromPosition(
       Position=&AddVectorToPosition(Position=$Particles.SkeletalMeshLocation.SampledPosition, Vector=<0 0 20>)
       Vector=Linked() {linkedVariable {name=Particles.Position, type=NiagaraPosition}}
      )
      Quaternion=<0 0 0 1>
     )
     Reference Axis Vector=<1 0 0>
    )) {isSetParameters=true}
   &ScaleColor(
    Scale Mode=::NewEnumerator1
    Scale RGBA=&MakeVector4_FromLinearColor(LinearColor=$System.TOD Color)
    ScaleRGBA=true
   ) {enabled=false, module=ScaleColor002}
  ]
  rendererClasses [&NiagaraMeshRendererProperties]
  renderers [
   {rendererClass=&NiagaraMeshRendererProperties, rendererIndex=0}
   {rendererClass=&NiagaraMeshRendererProperties, rendererIndex=1}
  ]
 }
 Lightning_Main_Blade {
  emitterUpdateScript [
   &EmitterState(
    Life Cycle Mode=::NewEnumerator1
    Loop Duration=&UniformRangedFloat(Minimum=0.08, Maximum=0.05)
    Recalculate Duration Each Loop=true
    UseLoopDelay=false
   )
   &SpawnBurst_Instantaneous(Spawn Count=2, Use Spawn Probability=false)
  ]
  particleSpawnScript [
   /Niagara/Modules/Spawn/Initialization/InitializeParticle(
    Color=&UniformRangedLinearColor(Minimum=<rgba 0.4 0.2969 0.321 1.25>, Maximum=<rgba 0.5687 0.6955 0.7 1.25>, Color Channel Mode=::NewEnumerator1)
    Lifetime=&UniformRangedFloat(Minimum=0.1, Maximum=0.16)
    Mesh Scale=&UniformRangedVector(Minimum=<-0.45 1 1>, Maximum=<-0.7 1 1>)
    Module.Write Color=true
    Module.Write Lifetime=true
    Module.Write Mass=false
    Module.Write Material Random=true
    Module.Write Position=false
    Module.Write Scale=true
    Module.Write SpriteRotation=false
    Module.Write SpriteSize=false
    Position=&SimulationPosition()
   ) {enabled=false}
   /Niagara/Modules/Spawn/Initialization/V2/InitializeParticle(
    Color=&MergeHelper_ColorSwitch(
     0=&UniformRangedLinearColor(Minimum=<rgba 1 0.3648 0 1.25>, Maximum=<rgba 1 0.191 0 1.25>, Color Channel Mode=::NewEnumerator1)
     1=&UniformRangedLinearColor(Minimum=<rgba 1 0.3648 0 1.25>, Maximum=<rgba 1 0.191 0 1.25>, Color Channel Mode=::NewEnumerator1)
     2=?
     3=?
     Index=&ReturnExecIndex()
     Num=::NewEnumerator0
    )
    Lifetime=&UniformRangedFloat(Minimum=0.05, Maximum=0.08)
    Mesh Scale=&UniformRangedVector(Minimum=<-0.35 1 1>, Maximum=<-0.6 1 1>)
    Mesh Scale Mode=::NewEnumerator4
    Position Offset=<0 0 30>
    UsePositionOffset=true
    Write Material Random=true
   ) {module=InitializeParticle001}
   &DynamicMaterialParameters(
    Index 0 Param 1=&UniformRangedFloat(Minimum=15, Maximum=10)
    Index 0 Param 2=&UniformRangedFloat(Minimum=0.2, Maximum=0.2)
    Index 0 Param 3=&UniformRangedFloat(Minimum=0, Maximum=0)
    Index 0 Param 4=&UniformRangedFloat(Minimum=0, Maximum=50)
    Index 1 Param 1=&UniformRangedFloat(Minimum=0.5, Maximum=1)
    Index 1 Param 2=&UniformRangedFloat(Minimum=0, Maximum=20)
    Module.Index0Float0Write=true
    Module.Index0Float1Write=true
    Module.Index0Float2Write=true
    Module.Index0Float3Write=true
    Module.Index1Float0Write=true
    Module.Index1Float1Write=true
    Module.Index1Float2Write=true
    Module.Index1Float3Write=true
    Write Parameter Index 1=::NewEnumerator1
   )
   &InitialMeshOrientation(Mesh Orientation Mode=::NewEnumerator2, Orientation Vector=<0 0 -1>, Use Rotation Vector=false)
   &ShapeLocation(MaskSpawnGroup=false, Override Seed=false, Sphere Radius=3, UseOwnerScale=true)
  ]
  particleUpdateScript [
   &ParticleState()
   &DynamicMaterialParameters(Module.Param0WriteEnabled=false, Module.Param1WriteEnabled=false, Module.Param2WriteEnabled=false, Module.Param3WriteEnabled=false) {enabled=false, module=DynamicMaterialParameters001}
   &ScaleColor(
    Scale Alpha=&FloatFromCurve(FloatCurve=DataInterface() {propertyValues {Curve {keys [<at 0.15 1>,<at 0.9 1>,<at 1 0>]}}}, Scale Curve=0.5)
    Scale RGB=&MakeVectorFromLinearColorRGB(Color=&MakeCustomLinearColorFromBool(
      Bool=&Boolean_OrOperation(Bool1=$User.bIsFrontEnd, Bool2=$User.bIsFrontEndPreview)
      True Color=&ColorFromCurve(ColorCurve=DataInterface() {propertyValues {
         AlphaCurve {keys []}
         BlueCurve {keys [<at 0 0>,<at 0.162 0>,<at 0.253 0>,<at 1 0>]}
         GreenCurve {keys [<at 0 1.3236>,<at 0.162 1.7999>,<at 0.253 0.6679>,<at 1 0.723>]}
         RedCurve {keys [<at 0 2.5>,<at 0.162 3.5>,<at 0.253 1.25>,<at 1 1>]}
        }})
      False Color=&ColorFromCurve(ColorCurve=DataInterface() {propertyValues {
         AlphaCurve {keys []}
         BlueCurve {keys [<at 0 2.5>,<at 0.162 3.5>,<at 0.253 1.25>,<at 1 1>]}
         GreenCurve {keys [<at 0 2.5>,<at 0.162 3.5>,<at 0.253 1.25>,<at 1 1>]}
         RedCurve {keys [<at 0 2.5>,<at 0.162 3.5>,<at 0.253 1.25>,<at 1 1>]}
        }})
     ))
    ScaleA=true
    ScaleRGB=true
   )
   &ScaleColor(
    Scale Mode=::NewEnumerator1
    Scale RGBA=&MakeVector4_FromLinearColor(LinearColor=$System.TOD Color)
    ScaleRGBA=true
   ) {enabled=false, module=ScaleColor002}
   &ScaleMeshSize(Scale Factor=&MakeVector(
     X=&FloatFromCurve(FloatCurve=DataInterface() {propertyValues {Curve {keys [<at 0 0.7 RCIM_Cubic>, {arriveTangent=0.15, leaveTangent=0.15, time=0.3, value=1}]}}})
     Y=1
     Z=1
    )) {enabled=false}
   /SetVariables(Particles.Position=&Add_Vector(
     A=Linked() {linkedVariable {name=Engine.Owner.Position, type=NiagaraPosition}}
     B=&RotateVectorByQuaternion(Vector To Rotate=<0 0 40>, Quaternion=<0 0 0 1>)
    )) {enabled=false, isSetParameters=true, module=SetVariables_#4}
  ]
  rendererClasses [&NiagaraMeshRendererProperties]
  renderers [{rendererClass=&NiagaraMeshRendererProperties, rendererIndex=0}]
 }
 Lightning_Main_Blade002 {
  emitterUpdateScript [
   &EmitterState(
    Life Cycle Mode=::NewEnumerator1
    Loop Duration=&UniformRangedFloat(Minimum=0.08, Maximum=0.1)
    Recalculate Duration Each Loop=true
    UseLoopDelay=false
   )
   &SpawnBurst_Instantaneous(Spawn Count=1, Use Spawn Probability=false)
  ]
  particleSpawnScript [
   /Niagara/Modules/Spawn/Initialization/InitializeParticle(
    Color=&UniformRangedLinearColor(Minimum=<rgba 0.4 0.2969 0.321 1.25>, Maximum=<rgba 0.5687 0.6955 0.7 1.25>, Color Channel Mode=::NewEnumerator1)
    Lifetime=&UniformRangedFloat(Minimum=0.1, Maximum=0.16)
    Mesh Scale=&UniformRangedVector(Minimum=<-0.45 1 1>, Maximum=<-0.7 1 1>)
    Module.Write Color=true
    Module.Write Lifetime=true
    Module.Write Mass=false
    Module.Write Material Random=true
    Module.Write Position=false
    Module.Write Scale=true
    Module.Write SpriteRotation=false
    Module.Write SpriteSize=false
    Position=&SimulationPosition()
   ) {enabled=false}
   /Niagara/Modules/Spawn/Initialization/V2/InitializeParticle(
    Color=&MergeHelper_ColorSwitch(
     0=&UniformRangedLinearColor(Minimum=<rgba 1.5 0.3289 0 1.25>, Maximum=<rgba 2 0.1999 0 1.25>, Color Channel Mode=::NewEnumerator1)
     1=&UniformRangedLinearColor(Minimum=<rgba 1 0.3648 0 1.25>, Maximum=<rgba 1 0.191 0 1.25>, Color Channel Mode=::NewEnumerator1)
     2=?
     3=?
     Index=&ReturnExecIndex()
     Num=::NewEnumerator0
    )
    Lifetime=&UniformRangedFloat(Minimum=0.08, Maximum=0.1)
    Mesh Scale=&UniformRangedVector(Minimum=<-0.35 1 1>, Maximum=<-0.6 1 1>)
    Mesh Scale Mode=::NewEnumerator4
    Position Offset=<0 0 30>
    UsePositionOffset=true
    Write Material Random=true
   ) {module=InitializeParticle001}
   &DynamicMaterialParameters(
    Index 0 Param 1=&UniformRangedFloat(Minimum=15, Maximum=20)
    Index 0 Param 2=&UniformRangedFloat(Minimum=0.2, Maximum=0.2)
    Index 0 Param 3=&UniformRangedFloat(Minimum=0, Maximum=0)
    Index 0 Param 4=&UniformRangedFloat(Minimum=0, Maximum=50)
    Index 1 Param 1=&UniformRangedFloat(Minimum=0.5, Maximum=1)
    Index 1 Param 2=&UniformRangedFloat(Minimum=0, Maximum=20)
    Module.Index0Float0Write=true
    Module.Index0Float1Write=true
    Module.Index0Float2Write=true
    Module.Index0Float3Write=true
    Module.Index1Float0Write=true
    Module.Index1Float1Write=true
    Module.Index1Float2Write=true
    Module.Index1Float3Write=true
    Write Parameter Index 1=::NewEnumerator1
   )
   &InitialMeshOrientation(Mesh Orientation Mode=::NewEnumerator2, Orientation Vector=<0 0 -1>, Use Rotation Vector=false)
   &ShapeLocation(MaskSpawnGroup=false, Override Seed=false, Sphere Radius=3, UseOwnerScale=true)
  ]
  particleUpdateScript [
   &ParticleState()
   &DynamicMaterialParameters(Module.Param0WriteEnabled=false, Module.Param1WriteEnabled=false, Module.Param2WriteEnabled=false, Module.Param3WriteEnabled=false) {enabled=false, module=DynamicMaterialParameters001}
   &ScaleColor(
    Scale Alpha=&FloatFromCurve(FloatCurve=DataInterface() {propertyValues {Curve {keys [<at 0.15 1>,<at 0.9 1>,<at 1 0>]}}})
    Scale RGB=&MakeVectorFromLinearColorRGB(Color=&MakeCustomLinearColorFromBool(
      Bool=&Boolean_OrOperation(Bool1=$User.bIsFrontEnd, Bool2=$User.bIsFrontEndPreview)
      True Color=&ColorFromCurve(ColorCurve=DataInterface() {propertyValues {
         AlphaCurve {keys []}
         BlueCurve {keys [<at 0 0>,<at 0.162 0>,<at 0.253 0>,<at 1 0>]}
         GreenCurve {keys [<at 0 1.3236>,<at 0.162 1.7999>,<at 0.253 0.6679>,<at 1 0.723>]}
         RedCurve {keys [<at 0 2.5>,<at 0.162 3.5>,<at 0.253 1.25>,<at 1 1>]}
        }})
      False Color=&ColorFromCurve(ColorCurve=DataInterface() {propertyValues {
         AlphaCurve {keys []}
         BlueCurve {keys [<at 0 2.5>,<at 0.162 3.5>,<at 0.253 1.25>,<at 1 1>]}
         GreenCurve {keys [<at 0 2.5>,<at 0.162 3.5>,<at 0.253 1.25>,<at 1 1>]}
         RedCurve {keys [<at 0 2.5>,<at 0.162 3.5>,<at 0.253 1.25>,<at 1 1>]}
        }})
     ))
    ScaleA=true
    ScaleRGB=true
   )
   &ScaleColor(
    Scale Mode=::NewEnumerator1
    Scale RGBA=&MakeVector4_FromLinearColor(LinearColor=$System.TOD Color)
    ScaleRGBA=true
   ) {enabled=false, module=ScaleColor002}
   &ScaleMeshSize(Scale Factor=&MakeVector(
     X=&FloatFromCurve(FloatCurve=DataInterface() {propertyValues {Curve {keys [<at 0 0.7 RCIM_Cubic>, {arriveTangent=0.15, leaveTangent=0.15, time=0.3, value=1}]}}})
     Y=1
     Z=1
    )) {enabled=false}
   /SetVariables(Particles.Position=&Add_Vector(
     A=Linked() {linkedVariable {name=Engine.Owner.Position, type=NiagaraPosition}}
     B=&RotateVectorByQuaternion(Vector To Rotate=<0 0 40>, Quaternion=<0 0 0 1>)
    )) {enabled=false, isSetParameters=true, module=SetVariables_#4}
  ]
  rendererClasses [&NiagaraMeshRendererProperties]
  renderers [{rendererClass=&NiagaraMeshRendererProperties, rendererIndex=0}]
 }
 MainCube {
  emitterUpdateScript [
   &EmitterState(Life Cycle Mode=::NewEnumerator1, Loop Delay=$System.Delay, Loop Duration=10, UseLoopDelay=true)
   &SpawnBurst_Instantaneous(Use Spawn Probability=false, UseLoopCountLimit=true)
  ]
  particleSpawnScript [
   /SetVariables(Particles.OutlineVis:NiagaraInt32=&MakeIntFromFloat(Float=$User.Variant, Conversion Mode=::NewEnumerator1)) {enabled=false, isSetParameters=true}
   &PartitionParticles(Number Of Partitions=16, Particles Per Partition=1, Partition Mode:ENiagara_PartitionMode_v2=::NewEnumerator1) {enabled=false}
   &InitializeParticle(
    Color=<rgba 0.3229 0.1646 0.0165 1>
    Lifetime=2
    Mesh Scale Mode=::NewEnumerator3
    Mesh Uniform Scale=0.65
    UsePositionOffset=false
    Write Material Random=true
   )
   &SpriteFacingAndAlignment(Sprite Alignment=<0 0 1>, WriteAlignment=true, WriteFacing=false)
  ]
  particleUpdateScript [
   &ParticleState(Kill Particles When Lifetime Has Elapsed=false)
   /SetVariables(Particles.OutlineColor:LinearColor=&Multiply_LinearColorByFloat(
     Linear Color=<rgba 4 4 4 1>
     Float=&FloatFromCurve(
      FloatCurve=DataInterface() {propertyValues {Curve {
         defaultValue=1
         keys [
          {interpMode=RCIM_Cubic, tangentMode=RCTM_User, time=0, value=1}
          {interpMode=RCIM_Cubic, tangentMode=RCTM_User, time=0.5, value=2}
          {tangentMode=RCTM_User, time=1, value=1}
         ]
        }}}
      Scale Curve=&FloatFromCurve(
       FloatCurve=DataInterface() {propertyValues {Curve {keys [{interpMode=RCIM_Cubic, tangentMode=RCTM_User, time=0, value=0}, <at 1 1>]}}}
       CurveIndex=$Emitter.Age
      )
     )
    )) {isSetParameters=true}
   &SkeletalMeshLocation(
    Bone / Socket Index=1
    Bone Sampling Mode=::NewEnumerator4
    MaskSpawnGroup=false
    Sampled Position Coordinate Space=::Local
    Skeletal Mesh=$System.Skeletal Mesh Reader
   )
   &ScaleMeshSize(Scale Factor=$Particles.SkeletalMeshLocation.SampledScale)
   &ScaleMeshSize(Scale Factor=&Multiply_VectorByFloat(
     Vector=<1 1 1>
     Float=&FloatFromCurve(FloatCurve=DataInterface() {propertyValues {Curve {keys [{arriveTangent=0.1632, interpMode=RCIM_Cubic, leaveTangent=0.1632, tangentMode=RCTM_User, time=0, value=0.9}, <at 1 1>]}}})
    )) {module=ScaleMeshSize001}
   /SetVariables(
    Particles.Glow_Base_CameraOffset=-5
    Particles.Glow_Base_Color:LinearColor=&Multiply_LinearColorByFloat(
     Linear Color=&MakeCustomLinearColorFromBool(
      Bool=&Boolean_OrOperation(Bool1=$User.bIsFrontEnd, Bool2=$User.bIsFrontEndPreview)
      True Color=<rgba 1 0.3925 0 0.457>
      False Color=<rgba 1 0.3136 0.0025 1>
     )
     Float=&Multiply_Float(
      A=&FNTimeOfDayScalar(Night Scale=0.6, Morning Scale=1, Evening Scale=1)
      B=&Multiply_Float(
       A=&Waveform(
        Global Input Position=&RandomRangeFloat(Minimum=0.5, Maximum=1)
        "Waveform [1]"=::NewEnumerator0
        "[1] Phase Scale"=5
        "[1] Frequency"=0.5
        "[1] Amplitude Min/Max"=<0.8 1>
       )
       B=&Waveform(
        Global Input Position=&RandomRangeFloat(Minimum=0, Maximum=0.5)
        "Waveform [1]"=::NewEnumerator0
        "[1] Phase Scale"=7
        "[1] Frequency"=0.5
        "[1] Amplitude Min/Max"=<0.8 1>
       )
      )
     )
    )
    Particles.Glow_Base_Size:Vector2f=&Multiply_Vector2D_ByFloat(
     Float=&SelectFloatFromArray(
      Direct Array Index=$Particles.PartitionParticles.Partition
      Float Selection Array=DataInterface() {propertyValues {FloatData [1,0.5,0.35,0.5,0.5,0.6,0.5,0.5,0.5,0,0,1,0,0.5,0.75,0.65]}}
     )
     Vector2D=<40 100>
    )
   ) {isSetParameters=true}
   &Light_Attributes(
    Color=1
    Color Value To Scale:LinearColor=/Niagara/DynamicInputs/LinearColor/MakeLinearColorFromVector4(Vector:Vector4f=<1 1 1 1>)
    Falloff Exponent=1
    Light Exponent=2
    Light Position Offset:Vector3f=<0 -8 0>
    Light Radius=40
    Position=1
    Radius=1
    Scale Alpha=10
    Scale RGB:Vector3f=&MakeVectorFromLinearColorRGB(Color=<rgba 1 0.3713 0 1>)
    ScaleA=true
    ScaleRGB=true
    UsePositionOffset=false
   )
  ]
  rendererClasses [
   &NiagaraMeshRendererProperties
   &NiagaraSpriteRendererProperties
   &NiagaraLightRendererProperties
  ]
  renderers [
   {rendererClass=&NiagaraMeshRendererProperties, rendererIndex=0}
   {rendererClass=&NiagaraSpriteRendererProperties, rendererIndex=1}
   {rendererClass=&NiagaraLightRendererProperties, rendererIndex=2}
  ]
 }
 SmallCubes {
  emitterUpdateScript [
   &EmitterState(Life Cycle Mode=::NewEnumerator1, Loop Behavior=::NewEnumerator1, Loop Duration Mode=::NewEnumerator1, UseLoopDelay=false)
   &SpawnRate(SpawnRate=10, Use Spawn Probability=false)
   &SpawnBurst_Instantaneous(Spawn Count=10, Use Spawn Probability=false, UseLoopCountLimit=false)
  ]
  particleSpawnScript [
   &InitializeParticle(
    Color Mode=::NewEnumerator0
    Lifetime Max=3
    Lifetime Min=0.2
    Lifetime Mode=::NewEnumerator1
    Mesh Scale Mode=::NewEnumerator1
    Mesh Uniform Scale Max=0.2
    Mesh Uniform Scale Min=0.6
    Position Offset=<0 0 27>
    Sprite Size Mode=::NewEnumerator3
    Uniform Sprite Size=11.4198
    UsePositionOffset=false
    Write Material Random=true
   )
   &SkeletalMeshLocation(Bone Sampling Mode=::NewEnumerator4, MaskSpawnGroup=false, Skeletal Mesh=$System.Skeletal Mesh Reader) {enabled=false}
   &SubUVAnimation(
    Mesh Renderer=DataInterface() {propertyValues {MeshRenderer=&self:SmallCubes_0.NiagaraMeshRendererProperties_0}}
    Output Alternate Renderer Binding=true
    Random Start Frame=true
    Renderer Type:ENiagaraMeshOrSprite=::NewEnumerator1
    SubUV Lookup Index Scale=0
    UseEndFrame=false
    UseStartFrame=false
   )
   &InitialMeshOrientation(Use Rotation Vector=false)
   &AddVelocity(
    Module.Use Falloff Distance=false
    Override Seed=false
    Velocity Mode=::NewEnumerator1
    Velocity Speed=&RandomRangeFloat(Minimum=9, Maximum=12)
   ) {enabled=false}
  ]
  particleUpdateScript [
   &ParticleState()
   &ShapeLocation(
    MaskSpawnGroup=false
    Non Uniform Scale=<1 1 1.3>
    Offset=<0 0 27>
    Offset Coordinate Space=::Local
    Offset Mode=::NewEnumerator0
    Override Seed=false
    Rotation Mode=::NewEnumerator2
    Sphere Radius=9
    UseOwnerScale=true
    Yaw / Pitch / Roll=&MakeVector(
     X=&Multiply_Float(A=$Engine.Time, B=&RandomRangeFloat(Minimum=0.1, Maximum=0.2))
     Y=0
     Z=0
    )
   )
   &ScaleMeshSize(Scale Factor=&VectorFromFloat(Value=&FloatFromCurve(FloatCurve=DataInterface() {propertyValues {Curve {
         defaultValue=1
         keys [
          {arriveTangent=6.6152, interpMode=RCIM_Cubic, leaveTangent=6.6152, tangentMode=RCTM_User, time=0, value=0}
          {arriveTangent=0.9862, interpMode=RCIM_Cubic, leaveTangent=0.9862, time=0.1033, value=0.9024}
          {arriveTangent=-1.0064, interpMode=RCIM_Cubic, leaveTangent=-1.0064, time=0.702, value=0.6923}
          {arriveTangent=-1.7524, interpMode=RCIM_Cubic, leaveTangent=-1.7524, tangentMode=RCTM_User, time=1, value=0}
         ]
        }}})))
   &OffsetPosition(Position Offset=&MakeVector(
     X=0
     Y=0
     Z=&Waveform(
      Global Phase=&Add_Float(A=$Emitter.Age, B=&RandomRangeFloat(Minimum=0, Maximum=1))
      "Waveform [1]"=::NewEnumerator0
      "[1] Phase Scale"=&RandomRangeFloat(Minimum=0.3, Maximum=0.5)
      "[1] Frequency"=&RandomRangeFloat(Minimum=0.5, Maximum=1)
      "[1] Amplitude Min/Max"=<0 1>
     )
    ))
   &UpdateMeshOrientation(Rotation Rate=&RandomRangeFloat(Minimum=0.1, Maximum=0.05), Rotation Vector=<1 1 1>)
   &Drag(
    Drag=&FloatFromCurve(
     FloatCurve=DataInterface() {propertyValues {Curve {
        defaultValue=1
        keys [
         {interpMode=RCIM_Cubic, tangentMode=RCTM_User, time=0, value=0}
         {arriveTangent=5.6533, interpMode=RCIM_Cubic, leaveTangent=5.6533, tangentMode=RCTM_User, time=0.223, value=1}
        ]
       }}}
     Scale Curve=&RandomRangeFloat(Minimum=0.5, Maximum=1)
    )
    Use Linear Drag=true
    Use Rotational Drag=true
   )
   &SolveForcesAndVelocity(Clamp Velocity=false, Limit Acceleration=false)
   /SetVariables(Particles.Position=&AddVectorToPosition(
     Position=$Particles.ShapeLocation.ShapePosition
     Vector=&TransformVector(
      Vector=&VectorFromCurve(
       VectorCurve=DataInterface() {propertyValues {XCurve {keys [<at 0 0>,<at 1 0>]}, YCurve {keys [<at 0 0>,<at 1 0>]}, ZCurve {keys [<at 0 0>,<at 1 1>]}}}
       Scale Curve=<1 1 20>
      )
      Destination Space=::World
     )
    )) {isSetParameters=true}
   &ScaleColor(
    Scale RGB=&MakeVectorFromLinearColorRGB(Color=&FNTimeOfDayColor(Night Color=<rgba 5 1.1796 0 1>, Morning Color=<rgba 2 0.3324 0 1>, Day Color=<rgba 2 0.3324 0 1>, Evening Color=<rgba 2 0.3324 0 1>))
    ScaleA=true
    ScaleRGB=true
   )
   /SetVariables(Particles.Linear Color=&ColorFromCurve(ColorCurve=DataInterface() {propertyValues {
       AlphaCurve {keys [
         <at 0 0>
         <at 0.1385 0.3109>
         {arriveTangent=-0.3609, interpMode=RCIM_Cubic, leaveTangent=-0.3609, time=0.7165, value=0.3118}
         <at 1 0 RCIM_Cubic>
        ]}
       BlueCurve {keys [<at 0 0 RCIM_Cubic>]}
       GreenCurve {keys [<at 0 0.3212 RCIM_Cubic>]}
       RedCurve {keys [<at 0 1 RCIM_Cubic>]}
      }})) {isSetParameters=true}
  ]
  rendererClasses [&NiagaraMeshRendererProperties, &NiagaraSpriteRendererProperties]
  renderers [
   {rendererClass=&NiagaraMeshRendererProperties, rendererIndex=0}
   {rendererClass=&NiagaraSpriteRendererProperties, rendererIndex=1}
  ]
 }
}
summary {
 systemName=NS_Pickaxe_VoyagerCuboid_Idle
 emitters [
  {emitterName=SmallCubes, rendererClasses [&NiagaraMeshRendererProperties, &NiagaraSpriteRendererProperties]}
  {
   emitterName=MainCube
   rendererClasses [
    &NiagaraMeshRendererProperties
    &NiagaraSpriteRendererProperties
    &NiagaraLightRendererProperties
   ]
  }
  {emitterName=Lightning_Main_Blade, rendererClasses [&NiagaraMeshRendererProperties]}
  {emitterName=Hunt_Lightning001, rendererClasses [&NiagaraMeshRendererProperties]}
  {emitterName=Lightning_Main_Blade002, rendererClasses [&NiagaraMeshRendererProperties]}
 ]
 userVariables [
  User.bIsFrontEnd {type=NiagaraBool}
  User.bIsFrontEndPreview {type=NiagaraBool}
  User.Variant {type=NiagaraFloat, defaultValue=0, description=""}
 ]
}
