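# Physics asset for the VoyagerCuboid pickaxe, giving physics bodies to 19 stone pieces on the axe mesh.
# Each piece pairs a tiny, identical capsule on a base_stone bone with an individually sized and oriented capsule on its dyn_stone bone.
# One constraint joins each dyn stone to its same-numbered base with swing and twist locked, and no constraints link different pairs.
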PhysicsAsset /BRCosmetics/Weapons/FORT_Melee/Pickaxe_VoyagerCuboid/Meshes/VoyagerCuboid_Axe_Physics
bodies {
 base_stone_01 [{shapeType=Capsule, length=0.505, radius=0.505, shapeName=base_stone_01_capsule, rotation=<rot 0 -90 0>}]
 base_stone_02 [{shapeType=Capsule, length=0.505, radius=0.505, shapeName=base_stone_02_capsule, rotation=<rot 0 -90 0>}]
 base_stone_03 [{shapeType=Capsule, length=0.505, radius=0.505, shapeName=base_stone_03_capsule, rotation=<rot 0 -90 0>}]
 base_stone_04 [{shapeType=Capsule, length=0.505, radius=0.505, shapeName=base_stone_04_capsule, rotation=<rot 0 -90 0>}]
 base_stone_05 [{shapeType=Capsule, length=0.505, radius=0.505, shapeName=base_stone_05_capsule, rotation=<rot 0 -90 0>}]
 base_stone_06 [{shapeType=Capsule, length=0.505, radius=0.505, shapeName=base_stone_06_capsule, rotation=<rot 0 -90 0>}]
 base_stone_07 [{shapeType=Capsule, length=0.505, radius=0.505, shapeName=base_stone_07_capsule, rotation=<rot 0 -90 0>}]
 base_stone_08 [{shapeType=Capsule, length=0.505, radius=0.505, shapeName=base_stone_08_capsule, rotation=<rot 0 -90 0>}]
 base_stone_09 [{shapeType=Capsule, length=0.505, radius=0.505, shapeName=base_stone_09_capsule, rotation=<rot 0 -90 0>}]
 base_stone_10 [{shapeType=Capsule, length=0.505, radius=0.505, shapeName=base_stone_10_capsule, rotation=<rot 0 -90 0>}]
 base_stone_11 [{shapeType=Capsule, length=0.505, radius=0.505, shapeName=base_stone_11_capsule, rotation=<rot 0 -90 0>}]
 base_stone_12 [{shapeType=Capsule, length=0.505, radius=0.505, shapeName=base_stone_12_capsule, rotation=<rot 0 -90 0>}]
 base_stone_13 [{shapeType=Capsule, length=0.505, radius=0.505, shapeName=base_stone_13_capsule, rotation=<rot 0 -90 0>}]
 base_stone_14 [{shapeType=Capsule, length=0.505, radius=0.505, shapeName=base_stone_14_capsule, rotation=<rot 0 -90 0>}]
 base_stone_15 [{shapeType=Capsule, length=0.505, radius=0.505, shapeName=base_stone_15_capsule, rotation=<rot 0 -90 0>}]
 base_stone_16 [{shapeType=Capsule, length=0.505, radius=0.505, shapeName=base_stone_16_capsule, rotation=<rot 0 -90 0>}]
 base_stone_17 [{shapeType=Capsule, length=0.505, radius=0.505, shapeName=base_stone_17_capsule, rotation=<rot 0 -90 0>}]
 base_stone_18 [{shapeType=Capsule, length=0.505, radius=0.505, shapeName=base_stone_18_capsule, rotation=<rot 0 -90 0>}]
 base_stone_19 [{shapeType=Capsule, length=0.505, radius=0.505, shapeName=base_stone_19_capsule, rotation=<rot 0 -90 0>}]
 dyn_stone_01 [{
   shapeType=Capsule
   length=5.1863
   radius=1.4522
   shapeName=dyn_stone_01_capsule
   center=<-0.0847 -0.3712 0.1403>
   rotation=<rot 0 77.7924 -53.5503>
  }]
 dyn_stone_02 [{
   shapeType=Capsule
   length=7.3426
   radius=3.3637
   shapeName=dyn_stone_02_capsule
   center=<-0.1626 -0.8874 -0.7708>
   rotation=<rot 0 82.6139 -75.3565>
  }]
 dyn_stone_03 [{
   shapeType=Capsule
   length=6.8159
   radius=1.3526
   shapeName=dyn_stone_03_capsule
   center=<-0.0374 -0.0612 0.0393>
   rotation=<rot 0 57.1828 -88.5038>
  }]
 dyn_stone_04 [{
   shapeType=Capsule
   length=4.2633
   radius=1.3004
   shapeName=dyn_stone_04_capsule
   center=<-0.1127 -0.0276 0.1759>
   rotation=<rot 0 54.8632 -69.0639>
  }]
 dyn_stone_05 [{
   shapeType=Capsule
   length=4.1544
   radius=1.6039
   shapeName=dyn_stone_05_capsule
   center=<0.1978 0.0655 0.1632>
   rotation=<rot 0 -100.8217 -59.1462>
  }]
 dyn_stone_06 [{
   shapeType=Capsule
   length=7.6497
   radius=3.0628
   shapeName=dyn_stone_06_capsule
   center=<0.2776 0.0294 0.196>
   rotation=<rot 0 -83.5135 -56.098>
  }]
 dyn_stone_07 [{
   shapeType=Capsule
   length=3.1158
   radius=0.9649
   shapeName=dyn_stone_07_capsule
   center=<-0.0322 0.0089 -0.2874>
   rotation=<rot 0 -118.1835 -71.6795>
  }]
 dyn_stone_08 [{
   shapeType=Capsule
   length=1.9494
   radius=0.8155
   shapeName=dyn_stone_08_capsule
   center=<0.0438 0.0807 0.0482>
   rotation=<rot 0 -119.526 -85.9672>
  }]
 dyn_stone_09 [{
   shapeType=Capsule
   length=6.6767
   radius=1.3345
   shapeName=dyn_stone_09_capsule
   center=<-0.0253 -0.0738 -0.4901>
   rotation=<rot 0 -52.4214 -71.5752>
  }]
 dyn_stone_10 [{
   shapeType=Capsule
   length=5.7331
   radius=1.5514
   shapeName=dyn_stone_10_capsule
   center=<-0.0266 -0.0983 -0.3376>
   rotation=<rot 0 -64.4217 -74.2805>
  }]
 dyn_stone_11 [{
   shapeType=Capsule
   length=7.0485
   radius=2.3908
   shapeName=dyn_stone_11_capsule
   center=<-0.4477 0.4956 0.3393>
   rotation=<rot 0 113.5761 -88.1571>
  }]
 dyn_stone_12 [{
   shapeType=Capsule
   length=4.4553
   radius=1.7054
   shapeName=dyn_stone_12_capsule
   center=<-0.137 0.6361 -0.1826>
   rotation=<rot 0 130.1155 -60.4411>
  }]
 dyn_stone_13 [{
   shapeType=Capsule
   length=2.7159
   radius=1.0955
   shapeName=dyn_stone_13_capsule
   center=<-0.0781 0.1638 0.1493>
   rotation=<rot 0 123.9482 -70.8562>
  }]
 dyn_stone_14 [{
   shapeType=Capsule
   length=3.286
   radius=2.4764
   shapeName=dyn_stone_14_capsule
   center=<-0.0419 0.037 0.0158>
   rotation=<rot 0 -151.8739 -65.7798>
  }]
 dyn_stone_15 [{
   shapeType=Capsule
   length=1.5267
   radius=0.7812
   shapeName=dyn_stone_15_capsule
   center=<-0.0021 0.0102 0.0657>
   rotation=<rot 0 113.6453 -48.6121>
  }]
 dyn_stone_16 [{
   shapeType=Capsule
   length=1.2999
   radius=1.0692
   shapeName=dyn_stone_16_capsule
   center=<0.0216 -0.0513 -0.0039>
   rotation=<rot 0 127.7614 -89.0793>
  }]
 dyn_stone_17 [{
   shapeType=Capsule
   length=2.0544
   radius=0.7796
   shapeName=dyn_stone_17_capsule
   center=<-0.003 -0.0249 0.1156>
   rotation=<rot 0 46.6599 -76.7096>
  }]
 dyn_stone_18 [{
   shapeType=Capsule
   length=4.6492
   radius=3.6283
   shapeName=dyn_stone_18_capsule
   center=<0.8134 0.096 -0.1712>
   rotation=<rot 0 -15.9052 -58.5268>
  }]
 dyn_stone_19 [{
   shapeType=Capsule
   length=1.5622
   radius=0.736
   shapeName=dyn_stone_19_capsule
   center=<-0.0769 -0.1148 -0.0497>
   rotation=<rot 0 -111.2408 -51.953>
  }]
}
constraints [
 {
  bone1Name=dyn_stone_01
  bone2Name=base_stone_01
  swing1LimitDegrees=45
  swing1Motion=Locked
  swing2LimitDegrees=45
  swing2Motion=Locked
  twistLimitDegrees=45
  twistMotion=Locked
 }
 {
  bone1Name=dyn_stone_02
  bone2Name=base_stone_02
  swing1LimitDegrees=45
  swing1Motion=Locked
  swing2LimitDegrees=45
  swing2Motion=Locked
  twistLimitDegrees=45
  twistMotion=Locked
 }
 {
  bone1Name=dyn_stone_03
  bone2Name=base_stone_03
  swing1LimitDegrees=45
  swing1Motion=Locked
  swing2LimitDegrees=45
  swing2Motion=Locked
  twistLimitDegrees=45
  twistMotion=Locked
 }
 {
  bone1Name=dyn_stone_04
  bone2Name=base_stone_04
  swing1LimitDegrees=45
  swing1Motion=Locked
  swing2LimitDegrees=45
  swing2Motion=Locked
  twistLimitDegrees=45
  twistMotion=Locked
 }
 {
  bone1Name=dyn_stone_05
  bone2Name=base_stone_05
  swing1LimitDegrees=45
  swing1Motion=Locked
  swing2LimitDegrees=45
  swing2Motion=Locked
  twistLimitDegrees=45
  twistMotion=Locked
 }
 {
  bone1Name=dyn_stone_06
  bone2Name=base_stone_06
  swing1LimitDegrees=45
  swing1Motion=Locked
  swing2LimitDegrees=45
  swing2Motion=Locked
  twistLimitDegrees=45
  twistMotion=Locked
 }
 {
  bone1Name=dyn_stone_07
  bone2Name=base_stone_07
  swing1LimitDegrees=45
  swing1Motion=Locked
  swing2LimitDegrees=45
  swing2Motion=Locked
  twistLimitDegrees=45
  twistMotion=Locked
 }
 {
  bone1Name=dyn_stone_08
  bone2Name=base_stone_08
  swing1LimitDegrees=45
  swing1Motion=Locked
  swing2LimitDegrees=45
  swing2Motion=Locked
  twistLimitDegrees=45
  twistMotion=Locked
 }
 {
  bone1Name=dyn_stone_09
  bone2Name=base_stone_09
  swing1LimitDegrees=45
  swing1Motion=Locked
  swing2LimitDegrees=45
  swing2Motion=Locked
  twistLimitDegrees=45
  twistMotion=Locked
 }
 {
  bone1Name=dyn_stone_10
  bone2Name=base_stone_10
  swing1LimitDegrees=45
  swing1Motion=Locked
  swing2LimitDegrees=45
  swing2Motion=Locked
  twistLimitDegrees=45
  twistMotion=Locked
 }
 {
  bone1Name=dyn_stone_11
  bone2Name=base_stone_11
  swing1LimitDegrees=45
  swing1Motion=Locked
  swing2LimitDegrees=45
  swing2Motion=Locked
  twistLimitDegrees=45
  twistMotion=Locked
 }
 {
  bone1Name=dyn_stone_12
  bone2Name=base_stone_12
  swing1LimitDegrees=45
  swing1Motion=Locked
  swing2LimitDegrees=45
  swing2Motion=Locked
  twistLimitDegrees=45
  twistMotion=Locked
 }
 {
  bone1Name=dyn_stone_13
  bone2Name=base_stone_13
  swing1LimitDegrees=45
  swing1Motion=Locked
  swing2LimitDegrees=45
  swing2Motion=Locked
  twistLimitDegrees=45
  twistMotion=Locked
 }
 {
  bone1Name=dyn_stone_14
  bone2Name=base_stone_14
  swing1LimitDegrees=45
  swing1Motion=Locked
  swing2LimitDegrees=45
  swing2Motion=Locked
  twistLimitDegrees=45
  twistMotion=Locked
 }
 {
  bone1Name=dyn_stone_15
  bone2Name=base_stone_15
  swing1LimitDegrees=45
  swing1Motion=Locked
  swing2LimitDegrees=45
  swing2Motion=Locked
  twistLimitDegrees=45
  twistMotion=Locked
 }
 {
  bone1Name=dyn_stone_16
  bone2Name=base_stone_16
  swing1LimitDegrees=45
  swing1Motion=Locked
  swing2LimitDegrees=45
  swing2Motion=Locked
  twistLimitDegrees=45
  twistMotion=Locked
 }
 {
  bone1Name=dyn_stone_17
  bone2Name=base_stone_17
  swing1LimitDegrees=45
  swing1Motion=Locked
  swing2LimitDegrees=45
  swing2Motion=Locked
  twistLimitDegrees=45
  twistMotion=Locked
 }
 {
  bone1Name=dyn_stone_18
  bone2Name=base_stone_18
  swing1LimitDegrees=45
  swing1Motion=Locked
  swing2LimitDegrees=45
  swing2Motion=Locked
  twistLimitDegrees=45
  twistMotion=Locked
 }
 {
  bone1Name=dyn_stone_19
  bone2Name=base_stone_19
  swing1LimitDegrees=45
  swing1Motion=Locked
  swing2LimitDegrees=45
  swing2Motion=Locked
  twistLimitDegrees=45
  twistMotion=Locked
 }
]
