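# Anim Blueprint for the VoyagerCuboid pickaxe mesh, adding procedural motion and physics to its floating-stone bones.
# Update computes an Ambient Rotation from real time for a Control Rig, plus five sine-driven translation vectors with no visible consumer.
# The rigged pose passes through 19 additive bone modifiers, shown with zero offsets, into a cached pose.
# RigidBody simulates that cache in base_stones_rot_m bone space with capped motion to produce the output.
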AnimBlueprint /BRCosmetics/Weapons/FORT_Melee/Pickaxe_VoyagerCuboid/Meshes/VoyagerCuboid_Axe_AnimBP
target_skeleton=&/VoyagerCuboid_Axe_Skeleton
graphs [
 AnimGraph {nodes [
   n1 = RigidBody(ComponentPose=Local To Component(LocalPose=Use cached pose 'CR_Cache_transforms'())) {
    bOverrideWorldGravity=true
    overridePhysicsAsset=&/VoyagerCuboid_Axe_Physics
    simulationSpace=BaseBoneSpace
    baseBoneRef {boneName=base_stones_rot_m}
    simSpaceSettings {
     maxAngularAcceleration=200
     maxAngularVelocity=3
     maxLinearAcceleration=2
     maxLinearVelocity=500
     worldAlpha=0.5
     externalLinearDragV=<0.2 0.2>
    }
   }
   Output Pose(Result=Component To Local(ComponentPose=@n1)) {layerGroup=None}
   AnimGraphNode_SubInput_0 = Input Pose()
   n2 = Local To Component(LocalPose=Control Rig!(Source=@AnimGraphNode_SubInput_0, Alpha, Ambient Rotation:Rotator=Get Ambient Rotation() -> (Ambient Rotation:Rotator)) -> (Pose) {controlRigAssetReference {controlRigAsset=&/VoyagerCuboid_Axe_CtrlRig}})
   n3 = "Transform (Modify) Bone - Bone: base_stone_03"(
    ComponentPose="Transform (Modify) Bone - Bone: base_stone_02"(
     ComponentPose="Transform (Modify) Bone - Bone: base_stone_01"(ComponentPose=@n2, Alpha=1, Translation="0,0,0", Rotation="0,0,0", Scale="1,1,1") {translationMode=BMM_Additive, translationSpace=BCS_BoneSpace, boneToModify {boneName=base_stone_01}}
     Alpha=1
     Translation="0,0,0"
     Rotation="0,0,0"
     Scale="1,1,1"
    ) {translationMode=BMM_Additive, translationSpace=BCS_BoneSpace, boneToModify {boneName=base_stone_02}}
    Alpha=1
    Translation="0,0,0"
    Rotation="0,0,0"
    Scale="1,1,1"
   ) {translationMode=BMM_Additive, translationSpace=BCS_BoneSpace, boneToModify {boneName=base_stone_03}}
   n4 = "Transform (Modify) Bone - Bone: base_stone_06"(
    ComponentPose="Transform (Modify) Bone - Bone: base_stone_05"(
     ComponentPose="Transform (Modify) Bone - Bone: base_stone_04"(ComponentPose=@n3, Alpha=1, Translation="0,0,0", Rotation="0,0,0", Scale="1,1,1") {translationMode=BMM_Additive, translationSpace=BCS_BoneSpace, boneToModify {boneName=base_stone_04}}
     Alpha=1
     Translation="0,0,0"
     Rotation="0,0,0"
     Scale="1,1,1"
    ) {translationMode=BMM_Additive, translationSpace=BCS_BoneSpace, boneToModify {boneName=base_stone_05}}
    Alpha=1
    Translation="0,0,0"
    Rotation="0,0,0"
    Scale="1,1,1"
   ) {translationMode=BMM_Additive, translationSpace=BCS_BoneSpace, boneToModify {boneName=base_stone_06}}
   n5 = "Transform (Modify) Bone - Bone: base_stone_09"(
    ComponentPose="Transform (Modify) Bone - Bone: base_stone_08"(
     ComponentPose="Transform (Modify) Bone - Bone: base_stone_07"(ComponentPose=@n4, Alpha=1, Translation="0,0,0", Rotation="0,0,0", Scale="1,1,1") {translationMode=BMM_Additive, translationSpace=BCS_BoneSpace, boneToModify {boneName=base_stone_07}}
     Alpha=1
     Translation="0,0,0"
     Rotation="0,0,0"
     Scale="1,1,1"
    ) {translationMode=BMM_Additive, translationSpace=BCS_BoneSpace, boneToModify {boneName=base_stone_08}}
    Alpha=1
    Translation="0,0,0"
    Rotation="0,0,0"
    Scale="1,1,1"
   ) {translationMode=BMM_Additive, translationSpace=BCS_BoneSpace, boneToModify {boneName=base_stone_09}}
   n6 = "Transform (Modify) Bone - Bone: base_stone_12"(
    ComponentPose="Transform (Modify) Bone - Bone: base_stone_11"(
     ComponentPose="Transform (Modify) Bone - Bone: base_stone_10"(ComponentPose=@n5, Alpha=1, Translation="0,0,0", Rotation="0,0,0", Scale="1,1,1") {translationMode=BMM_Additive, translationSpace=BCS_BoneSpace, boneToModify {boneName=base_stone_10}}
     Alpha=1
     Translation="0,0,0"
     Rotation="0,0,0"
     Scale="1,1,1"
    ) {translationMode=BMM_Additive, translationSpace=BCS_BoneSpace, boneToModify {boneName=base_stone_11}}
    Alpha=1
    Translation="0,0,0"
    Rotation="0,0,0"
    Scale="1,1,1"
   ) {translationMode=BMM_Additive, translationSpace=BCS_BoneSpace, boneToModify {boneName=base_stone_12}}
   n7 = "Transform (Modify) Bone - Bone: base_stone_15"(
    ComponentPose="Transform (Modify) Bone - Bone: base_stone_14"(
     ComponentPose="Transform (Modify) Bone - Bone: base_stone_13"(ComponentPose=@n6, Alpha=1, Translation="0,0,0", Rotation="0,0,0", Scale="1,1,1") {translationMode=BMM_Additive, translationSpace=BCS_BoneSpace, boneToModify {boneName=base_stone_13}}
     Alpha=1
     Translation="0,0,0"
     Rotation="0,0,0"
     Scale="1,1,1"
    ) {translationMode=BMM_Additive, translationSpace=BCS_BoneSpace, boneToModify {boneName=base_stone_14}}
    Alpha=1
    Translation="0,0,0"
    Rotation="0,0,0"
    Scale="1,1,1"
   ) {translationMode=BMM_Additive, translationSpace=BCS_BoneSpace, boneToModify {boneName=base_stone_15}}
   n8 = "Transform (Modify) Bone - Bone: base_stone_18"(
    ComponentPose="Transform (Modify) Bone - Bone: base_stone_17"(
     ComponentPose="Transform (Modify) Bone - Bone: base_stone_16"(ComponentPose=@n7, Alpha=1, Translation="0,0,0", Rotation="0,0,0", Scale="1,1,1") {translationMode=BMM_Additive, translationSpace=BCS_BoneSpace, boneToModify {boneName=base_stone_16}}
     Alpha=1
     Translation="0,0,0"
     Rotation="0,0,0"
     Scale="1,1,1"
    ) {translationMode=BMM_Additive, translationSpace=BCS_BoneSpace, boneToModify {boneName=base_stone_17}}
    Alpha=1
    Translation="0,0,0"
    Rotation="0,0,0"
    Scale="1,1,1"
   ) {translationMode=BMM_Additive, translationSpace=BCS_BoneSpace, boneToModify {boneName=base_stone_18}}
   Save cached pose 'CR_Cache_transforms'(Pose=Component To Local(ComponentPose="Transform (Modify) Bone - Bone: base_stone_19"(ComponentPose=@n8, Alpha=1, Translation="0,0,0", Rotation="0,0,0", Scale="1,1,1") {translationMode=BMM_Additive, translationSpace=BCS_BoneSpace, boneToModify {boneName=base_stone_19}})) {cacheName=CR_Cache_transforms}
  ]}
 EventGraph {nodes [
   Event BlueprintUpdateAnimation()
   n1 = GetRealTimeSeconds()
   n2 = float * float(A=float * float(A=2, B=@n1) {is=title K2Node_CommutativeAssociativeBinaryOperator}, B=0.35) {is=title K2Node_PromotableOperator}
   n3 = float * float(A=float * float(A=3, B=@n1) {is=title K2Node_CommutativeAssociativeBinaryOperator}, B=0.35) {is=title K2Node_PromotableOperator}
   n4 = float * float(A=float * float(A=5, B=@n1) {is=title K2Node_CommutativeAssociativeBinaryOperator}, B=0.35) {is=title K2Node_PromotableOperator}
   > Set Translate_Type_1(Translate_Type_1:Vector=MakeVector(X="Sin (Radians)"(A=@n4), Y="Sin (Radians)"(A=@n3), Z="Sin (Radians)"(A=@n2))) -> (Output_Get:Vector)
   n5 = GetRealTimeSeconds()
   n6 = float * float(A=float * float(A=1, B=@n5) {is=title K2Node_CommutativeAssociativeBinaryOperator}, B=0.3) {is=title K2Node_PromotableOperator}
   n7 = float * float(A=float * float(A=4, B=@n5) {is=title K2Node_CommutativeAssociativeBinaryOperator}, B=0.3) {is=title K2Node_PromotableOperator}
   n8 = "Sin (Radians)"(A=float * float(A=float * float(A=2, B=@n5) {is=title K2Node_CommutativeAssociativeBinaryOperator}, B=0.4) {is=title K2Node_PromotableOperator})
   > Set Translate_Type_2(Translate_Type_2:Vector=MakeVector(X=float * float(A=@n8, B=3) {is=title K2Node_PromotableOperator}, Y="Sin (Radians)"(A=@n7), Z="Sin (Radians)"(A=@n6))) -> (Output_Get:Vector)
   n9 = GetRealTimeSeconds()
   n10 = float * float(A=float * float(A=1, B=@n9) {is=title K2Node_CommutativeAssociativeBinaryOperator}, B=0.3)
   n11 = float * float(A=float * float(A=0, B=@n9) {is=title K2Node_CommutativeAssociativeBinaryOperator}, B=0.3) {is=title K2Node_PromotableOperator}
   n12 = float * float(A=float * float(A=3, B=@n9) {is=title K2Node_CommutativeAssociativeBinaryOperator}, B=0.3) {is=title K2Node_PromotableOperator}
   > Set Translate_Type_3(Translate_Type_3:Vector=MakeVector(X="Sin (Radians)"(A=@n12), Y="Sin (Radians)"(A=@n11), Z="Sin (Radians)"(A=@n10))) -> (Output_Get:Vector)
   n13 = GetRealTimeSeconds()
   n14 = float * float(A=float * float(A=3, B=@n13) {is=title K2Node_CommutativeAssociativeBinaryOperator}, B=0.3) {is=title K2Node_PromotableOperator}
   n15 = float * float(A=float * float(A=1, B=@n13) {is=title K2Node_CommutativeAssociativeBinaryOperator}, B=0.3) {is=title K2Node_PromotableOperator}
   n16 = float * float(A=float * float(A=4, B=@n13) {is=title K2Node_CommutativeAssociativeBinaryOperator}, B=0.3) {is=title K2Node_PromotableOperator}
   > Set Translate_Type_4(Translate_Type_4:Vector=MakeVector(X="Sin (Radians)"(A=@n16), Y="Sin (Radians)"(A=@n15), Z="Sin (Radians)"(A=@n14))) -> (Output_Get:Vector)
   n17 = GetRealTimeSeconds()
   n18 = float * float(A=float * float(A=1, B=@n17) {is=title K2Node_CommutativeAssociativeBinaryOperator}, B=0.3) {is=title K2Node_PromotableOperator}
   n19 = float * float(A=float * float(A=3, B=@n17) {is=title K2Node_CommutativeAssociativeBinaryOperator}, B=0.3) {is=title K2Node_PromotableOperator}
   n20 = float * float(A=float * float(A=3, B=@n17) {is=title K2Node_CommutativeAssociativeBinaryOperator}, B=0.3) {is=title K2Node_PromotableOperator}
   > Set Translate_Type_5(Translate_Type_5:Vector=MakeVector(X="Sin (Radians)"(A=@n18), Y="Sin (Radians)"(A=@n20), Z="Sin (Radians)"(A=@n19))) -> (Output_Get:Vector)
   n21 = float * float(A=float / float(A=GetRealTimeSeconds(), B=-720) {is=title K2Node_CallFunction}, B=93750)
   > Set Ambient Rotation(Ambient Rotation:Rotator=MakeRotator(Roll=@n21, Pitch=@n21, Yaw=@n21)) -> (Output_Get:Rotator)
   TryGetPawnOwner()
  ]}
]
variables [
 Ambient Rotation {type=Rotator}
 Rotate Shard Cube {type=Rotator}
 Translate_Type_1 {type=Vector, default=<-0.4765 0.8184 -0.9933>}
 Translate_Type_2 {type=Vector, default=<-0.9933 -0.229 0.6651>}
 Translate_Type_3 {type=Vector}
 Translate_Type_4 {type=Vector}
 Translate_Type_5 {type=Vector}
]
